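; Deployable storm item actor (DogSweater) that creates a rising, growing storm zone delivering ability effects through its sphere volume.
; Start fires gameplay cues and camera shake, enables delivery, sets a lifetime timer, then runs growth and lift timelines.
; The server drives scale and location while clients interpolate scale; audio attenuation follows size.
; A warning pulse precedes expiry, when replicated state triggers shrink and dissolve; death dissolves it early.
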(event Custom|SetRotation
  (params (PlayerYaw :type Rotator) (YawOffset :type Float%20%28double-precision%29))
  (Variables|Default|SetMeshLandRotation (Math|Rotator|MakeRotator 0.0 0.0 (.yaw PlayerYaw)))
  (Variables|Default|SetYawOffset YawOffset))

(event Custom|Start
  (bind _rootcomponent (Variables|Transformation|GetRootComponent))
  (bind _player_owner (Variables|Default|GetPlayer_Owner))
  (bind _returnvalue (Ability|GameplayCue|MakeGameplayCueParameters :NormalizedMagnitude 0.0 :RawMagnitude 0.0 :EffectContext "()" :AggregatedSourceTags "(GameplayTags=)" :AggregatedTargetTags "(GameplayTags=)" :Location "0,0,0" :Normal "0,0,0" :Instigator _player_owner :GameplayEffectLevel 1 :AbilityLevel 1 :TargetAttachComponent _rootcomponent :bReplicateLocationWhenUsingMinimalRepProxy false))
  ;[1 1 1] Start Sequence
  (Variables|Default|SetInitLocation (Transformation|GetActorLocation))
  (Utilities|FlowControl|Sequence
    (:then_0
      (Ability|ExecuteGameplayCueWithParamsReplicated (Variables|Default|GetChargeUpCue) _returnvalue)
      (Utilities|FlowControl|Delay (Variables|Default|GetGC_Delay))
      (Ability|ExecuteGameplayCueWithParamsReplicated (Variables|Default|GetZoneActivateCue) _returnvalue)
      (Camera|PlayWorldCameraShake "/Game/Effects/Fort_Effects/BluePrints/CameraShakes/B_Large_Explosion_CameraShake.B_Large_Explosion_CameraShake_C" (Transformation|GetActorLocation) 0.0 2e+03)
      (AbilitySystem|SetTouchComponentforAbilityDeliveryInfo (Variables|Default|GetStormVolumeSphere))
      (AbilitySystem|EnableAbilityDeliveryInfo (GameplayTags|MakeLiteralGameplayTag "(TagName=\"Ability.Athena.Item.DogSweater.DeliveryBucket\")"))
      (bind _returnvalue_2 (Utilities|Time|SetTimerbyFunctionName 0 "SetDissolve" (ScalableFloat|GetValueatLevel (Variables|Default|GetZoneLifeTime))))
      (Variables|Default|SetZoneLifeTimeHandle _returnvalue_2)
      (CallFunction|Grow))
    (:then_1
      (CallFunction|MoveUp))))

(event Custom|MoveUp
  (bind _initlocation (Variables|Default|GetInitLocation))
  (bind _x (.x _initlocation))
  (bind _y (.y _initlocation))
  (bind _z (.z _initlocation))
  ;[1 1 1] MoveUp
  (bind (_direction _newtrack_0) (Timeline|Timeline_1 Play)
    (:Update
      (if (Networking|IsServer)
        (Variables|Default|SetLocation (Math|Vector|MakeVector _x _y (+ _z _newtrack_0)))
        (else
          (Transformation|SetActorLocation (Math|Vector|MakeVector _x _y (+ _z _newtrack_0))))))))

(event Custom|Grow
  (bind _maxsize (Variables|Default|GetMaxSize))
  (bind _returnvalue (ScalableFloat|GetValueatLevel _maxsize))
  (bind _returnvalue_2 (Math|Vector|MakeVector _returnvalue _returnvalue _returnvalue))
  (bind _growthspeed (Variables|Default|GetGrowthSpeed))
  (bind _returnvalue_3 (ScalableFloat|GetValueatLevel _growthspeed))
  (bind _maxsize_2 (Variables|Default|GetMaxSize))
  (bind _returnvalue_4 (ScalableFloat|GetValueatLevel _maxsize_2))
  (bind _returnvalue_5 (Transformation|GetActorScale3D))
  (bind _returnvalue_6 (Utilities|Time|GetWorldDeltaSeconds))
  (bind _returnvalue_7 (Math|Interpolation|VInterptoConstant _returnvalue_5 _returnvalue_2 _returnvalue_6 _returnvalue_3))
  (bind _returnvalue_8 (Math|Vector|ClampVectorSize _returnvalue_7 0.0 _returnvalue_4))
  (bind _returnvalue_9 (Utilities|Time|GetWorldDeltaSeconds))
  (bind _returnvalue_10 (Transformation|GetActorScale3D))
  (bind _maxsize_3 (Variables|Default|GetMaxSize))
  (bind _returnvalue_11 (ScalableFloat|GetValueatLevel _maxsize_3))
  (bind _clientinterpspeed (Variables|Default|GetClientInterpSpeed))
  (bind _scale (Variables|Default|GetScale))
  (bind _returnvalue_12 (Math|Interpolation|VInterpTo _returnvalue_10 _scale _returnvalue_9 _clientinterpspeed))
  (bind _returnvalue_13 (Math|Vector|ClampVectorSize _returnvalue_12 0.0 _returnvalue_11))
  (Utilities|FlowControl|Sequence
    (:then_0
      (Transformation|SetActorScale3D "25.000000,25.000000,25.000000"))
    (:then_1
      (Variables|Default|SetTimelineIsPlaying true)
      ;[1 1 1] Grow
      (Timeline|Timeline_0 PlayFromStart
        (:Update
          (Utilities|FlowControl|Sequence
            (:then_0
              (if (Networking|IsServer)
                ;[1 1 1] Grow on Server
                (Transformation|SetActorScale3D _returnvalue_8)
                (Variables|Default|SetScale _returnvalue_8)
                (else
                  ;[1 1 1] Interp Towards Server Size
                  (Transformation|SetActorScale3D _returnvalue_13)))
              ;[1 1 1] Set Audio Progress Param
              (Audio|Parameter|SetFloatParameter (Variables|Default|GetGrowLoop) "Progress" (/ (Math|Vector|VectorLength _returnvalue_13) (ScalableFloat|GetValueatLevel (Variables|Default|GetMaxSize))) :!class /Script/AudioExtensions.AudioParameterControllerInterface)
              (Utilities|FlowControl|Sequence
                (:then_0
                  ;[1 1 1] Stop Growing Audio when reached max size
                  (if (Networking|IsServer)
                    (if (>= (Math|Vector|VectorLength (Transformation|GetActorScale3D)) (ScalableFloat|GetValueatLevel (Variables|Default|GetMaxSize)))
                      (CallFunction|StopGrowingAudio))))
                (:then_1
                  (if (Networking|IsServer)
                    (if (>= (/ (Utilities|Time|GetTimerElapsedTimebyHandle (Variables|Default|GetZoneLifeTimeHandle)) (ScalableFloat|GetValueatLevel (Variables|Default|GetZoneLifeTime))) (ScalableFloat|GetValueatLevel (Variables|Default|GetNearingEndWarningThreshold)))
                      (Utilities|FlowControl|DoOnce false
                        (:Completed
                          ;[1 1 1] Play Nearing End Sound
                          (CallFunction|PlayNearingEndSound))))))))
            (:then_1
              ;[1 1 1] Fade in ambient audio that grows with the sphere
              (Utilities|FlowControl|DoOnce false
                (:Completed
                  (Audio|Components|Audio|FadeIn (Variables|Default|GetAmbientAudioInside) 1.0 1.0 0.0 "Linear" :!class /Script/Engine.AudioComponent))))))))
    (:then_2
      ;[1 1 1] Match Client scale to server scale if they don't match.
      (if (Networking|IsServer)
        _))
    (:then_3
      ;[1 1 1] MoveProgressLineOnMaterial
      (if (not (Networking|IsServer))
        (bind (_direction _newtrack_0) (Timeline|Timeline_2 PlayFromStart)
          (:Update
            (Rendering|Material|SetScalarParameterValue (Variables|Default|GetDynamicMI) "Progress" (Math|Float|Lerp 0.5 -2e+01 _newtrack_0) :!class /Script/Engine.MaterialInstanceDynamic)))))
    (:then_4
      ;[1 1 1] Update Attenuation
      (if (Networking|IsServer)
        (CallFunction|SetAttenuation)
        (Utilities|Time|SetTimerbyFunctionName 0 "SetAttenuation" 0.2 true)))
    (:then_5
      (Audio|Components|Audio|Play (Variables|Default|GetGrowLoop) 0.0))))

(event Custom|StopGrowingAudio
  ;[1 1 1] Stop Growing Audio when reached max size
  (Audio|Components|Audio|FadeOut (Variables|Default|GetGrowLoop) 1.0 0.0 "Linear" :!class /Script/Engine.AudioComponent))

(event Custom|StartTimeline
  (bind _maxsize (Variables|Default|GetMaxSize))
  (bind _returnvalue (ScalableFloat|GetValueatLevel _maxsize))
  (bind _returnvalue_2 (Math|Vector|MakeVector _returnvalue _returnvalue _returnvalue))
  (bind _growthspeed (Variables|Default|GetGrowthSpeed))
  (bind _returnvalue_3 (ScalableFloat|GetValueatLevel _growthspeed))
  (bind _maxsize_2 (Variables|Default|GetMaxSize))
  (bind _returnvalue_4 (ScalableFloat|GetValueatLevel _maxsize_2))
  (bind _returnvalue_5 (Transformation|GetActorScale3D))
  (bind _returnvalue_6 (Utilities|Time|GetWorldDeltaSeconds))
  (bind _returnvalue_7 (Math|Interpolation|VInterptoConstant _returnvalue_5 _returnvalue_2 _returnvalue_6 _returnvalue_3))
  (bind _returnvalue_8 (Math|Vector|ClampVectorSize _returnvalue_7 0.0 _returnvalue_4))
  (bind _returnvalue_9 (Utilities|Time|GetWorldDeltaSeconds))
  (bind _returnvalue_10 (Transformation|GetActorScale3D))
  (bind _maxsize_3 (Variables|Default|GetMaxSize))
  (bind _returnvalue_11 (ScalableFloat|GetValueatLevel _maxsize_3))
  (bind _clientinterpspeed (Variables|Default|GetClientInterpSpeed))
  (bind _scale (Variables|Default|GetScale))
  (bind _returnvalue_12 (Math|Interpolation|VInterpTo _returnvalue_10 _scale _returnvalue_9 _clientinterpspeed))
  (bind _returnvalue_13 (Math|Vector|ClampVectorSize _returnvalue_12 0.0 _returnvalue_11))
  ;[1 1 1] Grow
  (Timeline|Timeline_0 PlayFromStart))

(event Custom|StopGrowth
  (bind _maxsize (Variables|Default|GetMaxSize))
  (bind _returnvalue (ScalableFloat|GetValueatLevel _maxsize))
  (bind _returnvalue_2 (Math|Vector|MakeVector _returnvalue _returnvalue _returnvalue))
  (bind _growthspeed (Variables|Default|GetGrowthSpeed))
  (bind _returnvalue_3 (ScalableFloat|GetValueatLevel _growthspeed))
  (bind _maxsize_2 (Variables|Default|GetMaxSize))
  (bind _returnvalue_4 (ScalableFloat|GetValueatLevel _maxsize_2))
  (bind _returnvalue_5 (Transformation|GetActorScale3D))
  (bind _returnvalue_6 (Utilities|Time|GetWorldDeltaSeconds))
  (bind _returnvalue_7 (Math|Interpolation|VInterptoConstant _returnvalue_5 _returnvalue_2 _returnvalue_6 _returnvalue_3))
  (bind _returnvalue_8 (Math|Vector|ClampVectorSize _returnvalue_7 0.0 _returnvalue_4))
  (bind _returnvalue_9 (Utilities|Time|GetWorldDeltaSeconds))
  (bind _returnvalue_10 (Transformation|GetActorScale3D))
  (bind _maxsize_3 (Variables|Default|GetMaxSize))
  (bind _returnvalue_11 (ScalableFloat|GetValueatLevel _maxsize_3))
  (bind _clientinterpspeed (Variables|Default|GetClientInterpSpeed))
  (bind _scale (Variables|Default|GetScale))
  (bind _returnvalue_12 (Math|Interpolation|VInterpTo _returnvalue_10 _scale _returnvalue_9 _clientinterpspeed))
  (bind _returnvalue_13 (Math|Vector|ClampVectorSize _returnvalue_12 0.0 _returnvalue_11))
  ;[1 1 1] Grow
  (Timeline|Timeline_0 Stop))

(event Custom|SetAttenuation
  (bind _returnvalue (Transformation|GetActorScale3D))
  (bind _returnvalue_2 (Math|Vector|VectorLength _returnvalue))
  (bind _returnvalue_3 (/ (* _returnvalue_2 (Variables|Sound|GetArbitraryFalloffMultiplier)) 2.0))
  (bind _arbitraryfalloffmultiplier (Variables|Sound|GetArbitraryFalloffMultiplier))
  (bind _returnvalue_4 (* _returnvalue_2 _arbitraryfalloffmultiplier))
  ;[1 1 1] Set Attenuation
  (Utilities|FlowControl|Sequence
    (:then_0
      (bind _structout (Utilities|Struct|SetmembersinSoundAttenuationSettings (Variables|Sound|GetAttenuationInside) (Math|Vector|MakeVector (/ _returnvalue_2 2.0))))
      (Audio|Components|Audio|AdjustAttenuation (Variables|Default|GetAmbientAudioInside) _structout))
    (:then_1
      (bind _structout_2 (Utilities|Struct|SetmembersinSoundAttenuationSettings (Variables|Sound|GetAttenuationOutside) _returnvalue_3 _returnvalue_4))
      (Audio|Components|Audio|AdjustAttenuation (Variables|Default|GetAmbientAudioOutside) _structout_2))
    (:then_2
      (bind _structout_3 (Utilities|Struct|SetmembersinSoundAttenuationSettings (Variables|Sound|GetAttenuationInsideEdge) _returnvalue_3 _returnvalue_4))
      (Audio|Components|Audio|AdjustAttenuation (Variables|Default|GetAmbientAudioInsideEdge) _structout_3))
    (:then_3
      (bind _structout_4 (Utilities|Struct|SetmembersinSoundAttenuationSettings (Variables|Sound|GetAttenuationTells) _returnvalue_4))
      (Audio|Components|Audio|AdjustAttenuation (Variables|Default|GetShrinkLoop) _structout_4))))

(event Custom|PlayNearingEndSound
  ;[1 1 1] Play Nearing End Sound
  (Utilities|FlowControl|Sequence
    (:then_0
      _)
    (:then_1
      (Utilities|FlowControl|Delay (* (- (ScalableFloat|GetValueatLevel (Variables|Default|GetZoneLifeTime)) (* (ScalableFloat|GetValueatLevel (Variables|Default|GetZoneLifeTime)) (ScalableFloat|GetValueatLevel (Variables|Default|GetNearingEndWarningThreshold)))) 0.5))))
  ;;[1 1 1] WarningSound
  (Audio|Components|Audio|Play (Variables|Default|GetNearingEnd) 0.0)
  (bind (_direction _newtrack_0) (Timeline|WarningPulse PlayFromStart)
    (:Update
      (Rendering|Material|SetScalarParameterValue (Variables|Default|GetDynamicMI) "Pulse" _newtrack_0 :!class /Script/Engine.MaterialInstanceDynamic))))

(event Custom|Destroy
  (AbilitySystem|DisableAbilityDeliveryInfo "(TagName=\"Ability.Athena.Item.DogSweater.DeliveryBucket\")")
  (Rendering|SetActorHiddenInGame true)
  (if (not (Networking|IsServer))
    (Audio|Components|Audio|FadeOut (Variables|Default|GetAmbientAudioOutside) 0.75 0.0 "Linear" :!class /Script/Engine.AudioComponent)
    (Actor|DestroyActor)))

(event Custom|CheckTick
  ;[1 1 1] Match Client scale to server scale if they don't match.
  (CallFunction|Multicast (Transformation|GetActorScale3D)))

(event Custom|Multicast
  (params (NewParam :type Vector))
  ;[1 1 1] Match Client scale to server scale if they don't match.
  (if (Networking|IsServer)
    _
    (elif (Math|Vector|Equal(Vector) NewParam (Transformation|GetActorScale3D) 1.0)
      _
      (else
        (Transformation|SetActorScale3D NewParam)
        (Development|PrintString "ForceSync" true true "(R=1.000000,G=0.000000,B=0.677986,A=1.000000)" 1.0)))))

(event Custom|SetDissolve
  ;[1 1 1] DissolveStorm
  (Variables|Default|SetTimelineIsPlaying false)
  (Variables|Default|SetIsDissolving true))

(event EventDestroyed
  (Audio|Components|Audio|FadeOut (Variables|Default|GetAmbientAudioOutside) 0.0 0.0 "Linear" :!class /Script/Engine.AudioComponent)
  (Ability|RemoveGameplayCueReplicated (Variables|Default|GetZoneActivateCue)))

(event Custom|StartDissolve
  (bind _dissolveplayrate (Variables|Default|GetDIssolvePlayRate))
  (bind _returnvalue (ScalableFloat|GetValueatLevel _dissolveplayrate))
  (bind _returnvalue_2 (Utilities|Time|GetWorldDeltaSeconds))
  (bind _returnvalue_3 (* _returnvalue 4e+02))
  (bind _zerovector (Variables|Default|GetZeroVector))
  (bind _scale (Variables|Default|GetScale))
  (bind _returnvalue_4 (Math|Interpolation|VInterptoConstant _scale _zerovector _returnvalue_2 _returnvalue_3))
  ;[1 1 1] DissolveStorm
  (Utilities|FlowControl|DoOnce false
    (:Completed
      (Audio|Components|Audio|Stop (Variables|Default|GetNearingEnd))
      (Components|Timeline|SetPlayRate (Variables|B_BGA_Athena_DogSweater|GetDissolveTimline) _returnvalue)
      (Utilities|FlowControl|Sequence
        (:then_0
          (bind (_direction _newtrack_0 _cloudfade) (Timeline|DissolveTimline Play)
            (:Update
              (Rendering|Material|SetScalarParameterValue (Variables|Default|GetDynamicMI) "DissolveAmount" _newtrack_0 :!class /Script/Engine.MaterialInstanceDynamic)
              (Rendering|Material|SetScalarParameterValueonMaterials (Variables|Default|GetSM_SphereInvertedDualLayer) "OverallOpacity" _cloudfade))
            (:Finished
              (CallFunction|Destroy))))
        (:then_1
          (Components|Timeline|Stop (Variables|B_BGA_Athena_DogSweater|GetWarningPulse))
          (bind (_direction_2 _newtrack_0_2) (Timeline|Pulse PlayFromStart)
            (:Update
              (Rendering|Material|SetScalarParameterValue (Variables|Default|GetDynamicMI) "Pulse" _newtrack_0_2 :!class /Script/Engine.MaterialInstanceDynamic))))
        (:then_2
          (Variables|Default|SetScale (Transformation|GetActorScale3D))
          (Timeline|Timeline_3 PlayFromStart
            (:Update
              (Variables|Default|SetScale _returnvalue_4)
              (Transformation|SetActorScale3D _returnvalue_4))))
        (:then_3
          (if (not (Networking|IsServer))
            (Audio|Components|Audio|Play (Variables|Default|GetShrinkBurst) 0.0)))))))

(event Custom|ClientStartDissolve
  (bind _dissolveplayrate (Variables|Default|GetDIssolvePlayRate))
  (bind _returnvalue (ScalableFloat|GetValueatLevel _dissolveplayrate))
  (bind _returnvalue_2 (Utilities|Time|GetWorldDeltaSeconds))
  (bind _returnvalue_3 (* _returnvalue 4e+02))
  (bind _zerovector (Variables|Default|GetZeroVector))
  (bind _scale (Variables|Default|GetScale))
  (bind _returnvalue_4 (Math|Interpolation|VInterptoConstant _scale _zerovector _returnvalue_2 _returnvalue_3))
  ;[1 1 1] DissolveStorm
  (Variables|Default|SetScale (Transformation|GetActorScale3D))
  (Timeline|Timeline_3 PlayFromStart))

(event Custom|AddLocalLoopingCue
  (params (Cue%20Tag :type Gameplay%20Tag%20Structure) (Cue%20Params :type Gameplay%20Cue%20Parameters%20Structure))
  (if (not (Networking|IsServer))
    (Projectile|AddGameplayCue%28Looping%29Local self Cue%20Tag Cue%20Params)))

(event OnComponentBeginOverlap(StormVolumeFX)
  (params (OverlappedComponent :type Primitive%20Component%20Object%20Reference) (OtherActor :type Actor%20Object%20Reference) (OtherComp :type Primitive%20Component%20Object%20Reference) (OtherBodyIndex :type Integer) (bFromSweep :type Boolean) (SweepResult :type Hit%20Result%20Structure%20%28by%20ref%29)))

(event Custom|RemoveLocalLoopingCue
  (params (Cue%20Tag :type Gameplay%20Tag%20Structure) (Cue%20Params :type Gameplay%20Cue%20Parameters%20Structure))
  (if (not (Networking|IsServer))
    (Projectile|RemoveGameplayCue%28Looping%29Local self Cue%20Tag Cue%20Params)))

(event OnComponentEndOverlap(StormVolumeFX)
  (params (OverlappedComponent :type Primitive%20Component%20Object%20Reference) (OtherActor :type Actor%20Object%20Reference) (OtherComp :type Primitive%20Component%20Object%20Reference) (OtherBodyIndex :type Integer)))

(event EventBeginPlay
  (bind _returnvalue (Game|GetInstigator))
  ;[1 1 1] On Load, added check to turn off smoke mesh on mobile
  (|CallParentFunction)
  (if (Fortnite|IsMobilePlatform)
    (Components|DestroyComponent (Variables|Default|GetSM_SphereInvertedDualLayer)))
  (bind _asfort_player_pawn (Utilities|Casting|CastToFortPlayerPawn (Variables|Actor|GetInstigator)))
  (Variables|Default|SetPlayer_Owner _asfort_player_pawn)
  (Utilities|Casting|CastToFortGameStateAthena (Game|GetGameState)
    (:then
      (Damage|SetAbilitySourceLevel 5))
    (:CastFailed
      (Damage|SetAbilitySourceLevel 5)))
  (Development|PrintString "DOgSweaterBeginPlay")
  (CallFunction|VFXSetTeamColors)
  (bind _returnvalue_2 (Rendering|Material|CreateDynamicMaterialInstance (Variables|Default|GetStormVolumeFX) 0 (Rendering|Material|GetMaterial (Variables|Default|GetStormVolumeFX) 0) "None" :!class /Script/Engine.PrimitiveComponent))
  (Variables|Default|SetDynamicMI _returnvalue_2)
  (EventDispatchers|BindEventtoOnDied (delegate Custom|DeadEvent))
  (if (Utilities|IsValid _returnvalue :!class /Script/Engine.KismetSystemLibrary)
    (CallFunction|SetTeam (Team|GetActorTeam _returnvalue))))

(event Custom|DeadEvent
  (params (DamagedActor :type Actor%20Object%20Reference) (Damage :type Float%20%28single-precision%29) (InstigatedBy :type Controller%20Object%20Reference) (DamageCauser :type Actor%20Object%20Reference) (HitLocation :type Vector) (FHitComponent :type Primitive%20Component%20Object%20Reference) (BoneName :type Name) (Momentum :type Vector))
  (Development|PrintString "Dead")
  (Networking|SwitchHasAuthority
    (:Authority
      (CallFunction|DestroyedDissolve)
      (CallFunction|StopGrowth))))

(event Custom|SetExplosion
  (params (Explosion%20Level :type Integer) (Activation_Radius :type Float%20%28double-precision%29) (PlayerPawn :type Fort%20Player%20Pawn%20Object%20Reference))
  (Variables|Default|SetActivation_Radius Activation_Radius)
  (Variables|Default|SetPlayer_Owner PlayerPawn)
  (Variables|Default|SetExplosionLevel Explosion%20Level))

(event Custom|DestroyedDissolve
  (Variables|Default|SetIsDissolving true)
  (Utilities|FlowControl|Gate false
    (:Exit
      (Utilities|FlowControl|Sequence
        (:then_0
          _)
        (:then_2
          _))
      (AbilitySystem|DisableAbilityDeliveryInfo "(TagName=\"Ability.Athena.Item.DogSweater.DeliveryBucket\")")
      (bind (_direction _newtrack_0) (Timeline|DestroyedDissolveTimeline PlayFromStart)
        (:Update
          (Rendering|Material|SetScalarParameterValue (Variables|Default|GetDynamicMI) "DissolveAmount" _newtrack_0 :!class /Script/Engine.MaterialInstanceDynamic))
        (:Finished
          (Actor|DestroyActor))))))

(fn AttachFX
  (Networking|CanExecuteCosmeticEvents))

(fn DestroyFX)

(fn IsValidTarget
  (params (Object :type Object%20Reference))
  (if (Variables|Default|GetbActivated)
    (bind _aspawn (Utilities|Casting|CastToPawn Object)
      (:then
        (switch Utilities|FlowControl|Switch|SwitchonEFortTeamAffiliation (Team|GetTeamAffiliation _aspawn self)
          (:Friendly
            (Variables|Default|SetIsValidEnemy false))
          (:Neutral
            (Variables|Default|SetIsValidEnemy true))
          (:Hostile
            _))
        (return (Variables|Default|GetIsValidEnemy)))
      (:CastFailed
        _))
    (else
      _)))

(fn OnRep_bActivated)

(fn OnRep_ExplosionRadiusGuide)

(fn OnRep_IsDead?)

(fn OnRep_IsDissolving
  (CallFunction|StartDissolve)
  (if (Networking|IsServer)
    (CallFunction|CheckTick)))

(fn OnRep_IsPlaced)

(fn OnRep_Location
  (Transformation|SetActorLocation (Variables|Default|GetLocation)))

(fn OnRep_SavedScale
  (Transformation|SetActorScale3D (Variables|Default|GetSavedScale)))

(fn OnRep_Scale)

(fn OnRep_TimelineIsPlaying
  (if (Variables|Default|GetTimelineIsPlaying)
    (CallFunction|StartTimeline)
    (if (Networking|IsServer)
      (CallFunction|CheckTick))))

(fn ConstructionScript)

(fn VFX_SetTeamColors)
