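; Thrown projectile for the DogSweater deployable storm consumable, which animates its top part's rotation and scale from spawn.
; Hits are filtered by surface slope, with a velocity bounce for player pawns and NoStick targets; vehicle attachment is only a debug print.
; Stopping or entering water disables collision, then the server spawns the storm building gameplay actor at a Z offset and sends it a gameplay event.
; Non-server instances remove the mesh and destroy themselves after a delay.
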(event EventBeginPlay
  (bind _top (Variables|Default|GetTop))
  ;[1 1 1] Rotate and scale over time
  (Components|Activation|Activate (Variables|Default|GetInAirAudioLoop) false)
  (Components|Activation|Activate (Variables|Default|GetP_DeployableStorm_Trail) false)
  (bind (_direction _rotatoralpha) (Timeline|Rotate PlayFromStart)
    (:Update
      (Transformation|SetRelativeRotation _top (Math|Rotator|MakeRotator (.roll (Variables|/Script/Engine.SceneComponent|GetRelativeRotation (Variables|Default|GetTop))) (Math|Float|Lerp 3e+01 9e+01 _rotatoralpha) (.yaw (Variables|/Script/Engine.SceneComponent|GetRelativeRotation (Variables|Default|GetTop)))) false false)
      (Transformation|SetRelativeScale3D _top (Math|Conversions|ToVector(Float) (Math|Float|Lerp 0.5 0.7 _rotatoralpha))))))

(event OnEnterWaterEvent(WaterInteractionComponent)
  (params (WaterBody :type Fort%20Water%20Body%20Actor%20Object%20Reference) (WaterInteractionComponent :type Fort%20Water%20Interaction%20Component%20Object%20Reference) (bIsFirstBody :type Boolean))
  ;[1 1 1] On Stop Spawn the Storm Flip
  (Utilities|FlowControl|DoOnce false
    (:Completed
      (Fortnite|ShowSoundIndicatorLocal :Location (Transformation|GetActorLocation) :MaxAudibleDistance 2.75e+03 :IndicatorType "Gunshot" :Instigator self :Tint "(R=0.000000,G=0.000000,B=0.000000,A=0.000000)" :AttenuationMaxRangeMultiplier 1.0 :AttenuationIntensityMultiplier 1.0)
      (Audio|Components|Audio|FadeOut (Variables|Default|GetInAirAudioLoop) 0.3 0.0 "Linear" :!class /Script/Engine.AudioComponent)
      (Collision|SetActorEnableCollision false)
      (if (Networking|IsServer)
        (bind _returnvalue (Building|SpawnBuildingGameplayActor "/Game/Athena/Items/Consumables/DeployableStorm/B_BGA_Athena_DogSweater.B_BGA_Athena_DogSweater_C" (Math|Transform|MakeTransform (+ (Transformation|GetActorLocation) (Variables|Default|GetPlacementZOffset))) (Variables|Actor|GetInstigator)))
        (Development|PrintString "SENTEVENT")
        (Utilities|FlowControl|Delay 0.0)
        (Ability|SendGameplayEventtoActor _returnvalue (Variables|Default|GetEventComplete) (Utilities|Struct|MakeGameplayEventData :OptionalObject _returnvalue :ContextHandle "()" :InstigatorTags "(GameplayTags=)" :TargetTags "(GameplayTags=)" :EventMagnitude 0.0 :TargetData "()"))
        (Audio|Components|Audio|FadeOut (Variables|Default|GetInAirAudioLoop) 0.2 0.0 "Linear" :!class /Script/Engine.AudioComponent)
        (else
          (Components|DestroyComponent (Variables|Default|GetDogSweater_ProtoMesh))
          (Utilities|Time|SetTimerbyEvent (delegate Custom|DestryActor) (Variables|Default|GetDestroyDelay))
          (Components|Activation|Deactivate (Variables|Default|GetFX_DeployableStorm_Start)))))))

(event EventOnStop
  (params (Hit :type Hit%20Result%20Structure%20%28by%20ref%29))
  ;[1 1 1] On Stop Spawn the Storm Flip
  (Utilities|FlowControl|DoOnce false
    (:Completed
      (Fortnite|ShowSoundIndicatorLocal :Location (Transformation|GetActorLocation) :MaxAudibleDistance 2.75e+03 :IndicatorType "Gunshot" :Instigator self :Tint "(R=0.000000,G=0.000000,B=0.000000,A=0.000000)" :AttenuationMaxRangeMultiplier 1.0 :AttenuationIntensityMultiplier 1.0)
      (Audio|Components|Audio|FadeOut (Variables|Default|GetInAirAudioLoop) 0.3 0.0 "Linear" :!class /Script/Engine.AudioComponent)
      (Collision|SetActorEnableCollision false)
      (if (Networking|IsServer)
        (bind _returnvalue (Building|SpawnBuildingGameplayActor "/Game/Athena/Items/Consumables/DeployableStorm/B_BGA_Athena_DogSweater.B_BGA_Athena_DogSweater_C" (Math|Transform|MakeTransform (+ (Transformation|GetActorLocation) (Variables|Default|GetPlacementZOffset))) (Variables|Actor|GetInstigator)))
        (Development|PrintString "SENTEVENT")
        (Utilities|FlowControl|Delay 0.0)
        (Ability|SendGameplayEventtoActor _returnvalue (Variables|Default|GetEventComplete) (Utilities|Struct|MakeGameplayEventData :OptionalObject _returnvalue :ContextHandle "()" :InstigatorTags "(GameplayTags=)" :TargetTags "(GameplayTags=)" :EventMagnitude 0.0 :TargetData "()"))
        (Audio|Components|Audio|FadeOut (Variables|Default|GetInAirAudioLoop) 0.2 0.0 "Linear" :!class /Script/Engine.AudioComponent)
        (else
          (Components|DestroyComponent (Variables|Default|GetDogSweater_ProtoMesh))
          (Utilities|Time|SetTimerbyEvent (delegate Custom|DestryActor) (Variables|Default|GetDestroyDelay))
          (Components|Activation|Deactivate (Variables|Default|GetFX_DeployableStorm_Start)))))))

(event Custom|DestryActor
  ;[1 1 1] On Stop Spawn the Storm Flip
  (Actor|DestroyActor))

(event Custom|SetRotation
  (params (PlayerRotation :type Rotator) (YawOffset :type Float%20%28double-precision%29))
  ;[1 1 1] Store location
  (Variables|Default|SetPlacementRotation PlayerRotation)
  (Variables|Default|SetPlacementRotationYawOffset YawOffset))

(event Collision|EventHit
  (params (MyComp :type Primitive%20Component%20Object%20Reference) (Other :type Actor%20Object%20Reference) (OtherComp :type Primitive%20Component%20Object%20Reference) (bSelfMoved :type Boolean) (HitLocation :type Vector :default "0, 0, 0") (HitNormal :type Vector :default "0, 0, 0") (NormalImpulse :type Vector :default "0, 0, 0") (Hit :type Hit%20Result%20Structure%20%28by%20ref%29))
  (bind (_bblockinghit _binitialoverlap _time _distance _location _impactpoint _normal _impactnormal _physmat _hitactor _hitcomponent) (Collision|BreakHitResult Hit))
  (bind (_asfort_player_pawn _bsuccess) (Utilities|Casting|CastToFortPlayerPawn _hitactor))
  (bind (_asfort_vehicle_interface _bsuccess_2) (Utilities|Casting|CastToFortVehicleInterface _hitactor))
  ;[1 1 1] Check if we hit a flat enough surface.
  (if (> (.z HitNormal) (Variables|Default|GetWorldStopSlop))
    (if (Math|Boolean|ORBoolean _bsuccess (Actor|ActorHasTag _hitactor "NoStick") (Components|ComponentHasTag _hitcomponent "NoStick"))
      (if (not _bsuccess_2)
        ;[1 1 1] If you hit a target that isn't valid
        (Variables|/Script/Engine.MovementComponent|SetVelocity (Math|Vector|vector+vector (Math|Vector|MirrorVectorbyNormal (* (Variables|/Script/Engine.MovementComponent|GetVelocity (Variables|Movement|GetProjectileMovementComponent)) 0.75) "0, 0,1.000000") (Math|Vector|MakeVector 0.0 0.0 (* (Math|Vector|VectorLength (Variables|/Script/Engine.MovementComponent|GetVelocity (Variables|Movement|GetProjectileMovementComponent))) 0.25))) (Variables|Movement|GetProjectileMovementComponent))
        (elif (and (> (.z HitNormal) (Variables|Default|GetVehicleStopSlope)) _bsuccess_2)
          (Development|PrintString "AttachVehicle")))
      (else
        _))
    (Game|Components|ProjectileMovement|StopSimulating (Variables|Movement|GetProjectileMovementComponent) Hit)
    (else
      _)))

(fn ConstructionScript)
