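; Gameplay ability for the Deployable Storm (DogSweater) consumable that clears the effect that granted it, by name a damage effect, once its owner leaves the damage zone.
; On activation it waits for the InDamageZone gameplay tag to be removed, then ends itself.
; Ending the ability, whether cancelled or not, calls RemoveGrantedbyEffect, so the granting GameplayEffect's lifetime follows that tag.
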(event Ability|EventActivateAbility
  (Ability|Tasks|WaitGameplayTagRemove "(TagName=\"Granted.Athena.Item.DogSweater.InDamageZone\")"
    (:Removed
      (Ability|EndAbility))))

(event Ability|EventOnEndAbility
  (params (bWasCancelled :type Boolean))
  (Ability|RemoveGrantedbyEffect))
