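; Secondary trajectory ability for the DogSweater deployable storm consumable, layering a "cook" effect onto its parent ability.
; On activation it runs the parent logic, then applies the DogSweater Cook gameplay effect to the avatar actor as a global environment effect.
; On end it runs the parent logic, then removes effects carrying the DogSweater Cook asset tag from the owner, whether or not the ability was cancelled.
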(event Ability|EventActivateAbility
  (|CallParentFunction)
  (Environment|ApplyGlobalEnvironmentGameplayEffecttoActor (Ability|GetAvatarActorfromActorInfo) "/Game/Athena/Items/Consumables/DeployableStorm/GE_Athena_DogSweater_Cook.GE_Athena_DogSweater_Cook_C"))

(event Ability|EventOnEndAbility
  (params (bWasCancelled :type Boolean))
  (|CallParentFunction bWasCancelled)
  (Ability|RemoveGameplayEffectFromOwnerWithAssetTags "(GameplayTags=((TagName=\"Asset.Athena.Item.DogSweater.Cook\")))"))
